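# Sound cue for dropping the Season 14 Halloween Candy Corn item.
# A single wave player plays the CandyCorn_Halloween_Drop_01 sound, with no randomization or modulation.
# It uses the shared weapon-drop attenuation, with a limited max distance, and routes through the loot-collect sound class.
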SoundCue /Game/Athena/Sounds/Items/S14_Halloween/CandyCorn_Halloween_Drop_Cue
root=WavePlayer_0
nodes [{id=WavePlayer_0, properties {soundWaveAssetPtr=&/CandyCorn_Halloween_Drop_01}}]
properties {
 attenuationSettings=&Fort_Weapon_Drop_Att
 maxDistance=3000
 soundClassObject=&Fort_Loot_Collect
}
