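# Sound cue for picking up Halloween candy corn, an Athena item in the S14_Halloween set.
# A single wave player plays one pickup wave directly, with no randomization or modulation.
# The cue uses the weapon pickup attenuation asset and the loot collect sound class, at slightly reduced volume.
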SoundCue /Game/Athena/Sounds/Items/S14_Halloween/CandyCorn_Halloween_Pickup_Cue
root=WavePlayer_2
nodes [{id=WavePlayer_2, properties {soundWaveAssetPtr=&/CandyCorn_Halloween_Pickup_01}}]
properties {
 attenuationSettings=&Fort_Weapon_Pickup_Att
 maxDistance=3500
 soundClassObject=&Fort_Loot_Collect
 volumeMultiplier=0.7
}
