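# Sound cue for picking up the Halloween jelly bean item, using pickup attenuation and the loot-collect sound class.
# A mixer layers the jelly bean pickup wave with a boosted copy of the mini shield potion completion sound.
# The potion layer is pitched up by a fixed modulator, then enveloped to stay near-silent briefly before rising quickly, masking its opening.
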SoundCue /Game/Athena/Sounds/Items/S14_Halloween/JellyBean_Halloween_Pickup_Cue
root=Mixer_0
nodes [
 {
  id=Enveloper_0
  children [Modulator_0]
  properties {pitchCurve {editorCurveData {keys [{time=0}]}}, volumeCurve {editorCurveData {keys [<at 0.3687 0.0124>,<at 0.4486 1.0007>]}}}
 }
 {id=Mixer_0, children [WavePlayer_2,Enveloper_0], properties {inputVolume [1,4.4]}}
 {id=Modulator_0, children [WavePlayer_0], properties {pitchMax=1.6, pitchMin=1.6, volumeMax=1, volumeMin=1}}
 {id=WavePlayer_0, properties {soundWaveAssetPtr=/Game/Athena/Sounds/Items/sfx_MiniShieldPotion_Complete_01}}
 {id=WavePlayer_2, properties {soundWaveAssetPtr=&/JellyBean_Halloween_Pickup_01}}
]
properties {
 attenuationSettings=&Fort_Weapon_Pickup_Att
 maxDistance=3500
 soundClassObject=&Fort_Loot_Collect
 volumeMultiplier=0.55
}
