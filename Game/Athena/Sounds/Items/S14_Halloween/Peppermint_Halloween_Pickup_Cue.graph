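# Sound cue for picking up the Season 14 Halloween peppermint item, routed to the loot collect sound class.
# A mixer layers the peppermint pickup wave with a briefly delayed pepper boost loop wave borrowed from the JungleForaged items; its third input is empty.
# An enveloper fades the boost layer out over about a second and a half with slightly raised pitch.
# Weapon pickup attenuation is assigned, with an override enabled.
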SoundCue /Game/Athena/Sounds/Items/S14_Halloween/Peppermint_Halloween_Pickup_Cue
root=Mixer_0
nodes [
 {id=Delay_1, children [Enveloper_1], properties {delayMax=0.2, delayMin=0.2}}
 {
  id=Enveloper_1
  children [WavePlayer_1]
  properties {
   pitchCurve {editorCurveData {keys [<at 0.0001 1.0719>]}}
   volumeCurve {editorCurveData {keys [<at 0 1>,<at 0.3778 0.9972>,<at 0.6712 0.5624>,<at 1.4928 -0.0474>]}}
  }
 }
 {id=Mixer_0, children [WavePlayer_2,Delay_1,null], properties {inputVolume [0.7,1,1.65]}}
 {id=WavePlayer_1, properties {soundWaveAssetPtr=/Game/Athena/Sounds/Items/JungleForaged/Athena_Pepper_BoostLoop_01}}
 {id=WavePlayer_2, properties {soundWaveAssetPtr=&/Peppermint_Halloween_Pickup_01}}
]
properties {
 attenuationSettings=&Fort_Weapon_Pickup_Att
 bOverrideAttenuation=true
 maxDistance=3200
 soundClassObject=&Fort_Loot_Collect
 volumeMultiplier=1
 attenuationOverrides {falloffDistance=2400, stereoSpread=4000, attenuationShapeExtents {x=800}}
}
