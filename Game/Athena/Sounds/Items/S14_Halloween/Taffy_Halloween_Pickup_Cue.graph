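# Sound cue for picking up the Halloween taffy item, classed as loot collect with weapon pickup attenuation.
# A mixer layers the main taffy pickup wave with two quieter, briefly delayed accents.
# The first accent is a taffy thermal wave, and the later, quietest one reuses the Flopper thermal fish activation wave pitched up.
# All timings and modulator values are fixed, so playback has no random variation.
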SoundCue /Game/Athena/Sounds/Items/S14_Halloween/Taffy_Halloween_Pickup_Cue
root=Mixer_0
nodes [
 {id=Delay_0, children [Modulator_1], properties {delayMax=0.1, delayMin=0.1}}
 {id=Delay_1, children [Modulator_0], properties {delayMax=0.25, delayMin=0.25}}
 {id=Mixer_0, children [WavePlayer_2,Delay_0,Delay_1], properties {inputVolume [1,0.7,0.5]}}
 {id=Modulator_0, children [WavePlayer_1], properties {pitchMax=1.3, pitchMin=1.3, volumeMax=1, volumeMin=1}}
 {id=Modulator_1, children [WavePlayer_0], properties {pitchMax=1, pitchMin=1, volumeMax=1, volumeMin=1}}
 {id=WavePlayer_0, properties {soundWaveAssetPtr=&/Taffy_Halloween_Pickup_Thermal_01}}
 {id=WavePlayer_1, properties {soundWaveAssetPtr=/Game/Athena/Sounds/Items/Flopper/Thermal_Fish_Acxtivate}}
 {id=WavePlayer_2, properties {soundWaveAssetPtr=&/Taffy_Halloween_Pickup_01}}
]
properties {
 attenuationSettings=&Fort_Weapon_Pickup_Att
 maxDistance=3500
 soundClassObject=&Fort_Loot_Collect
 volumeMultiplier=0.4
}
